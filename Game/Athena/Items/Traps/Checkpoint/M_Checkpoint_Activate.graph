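# Translucent unlit particle material for the checkpoint trap's activation effect, projecting a colored block pattern onto scene surfaces behind it.
# A world-aligned texture at the behind-translucency position, with sine-modulated blocks, combines with a scene-normal versus vertex-normal edge term to drive color and opacity.
# Radial or linear sweep, depth fade and optional square mask shape opacity; particle color tints it.
# Only deferred shading shows the effect; other paths output zero.
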Material /Game/Athena/Items/Traps/Checkpoint/M_Checkpoint_Activate
graph {
 properties {
  bUsedWithMeshParticles=true
  bUsedWithNiagaraSprites=true
  bUsedWithParticleSprites=true
  blendMode=BLEND_Translucent
  pixelDepthOffsetMode=PDOM_Legacy
  shadingModel=MSM_Unlit
  usedShadingModels=Unlit
 }
 nodes [
  n1 = Constant()
  n2 = ComponentMask(None=SceneTexture() {sceneTextureId=PPI_WorldNormal}.Color) {b=true, g=true, r=true}
  n3 = VertexNormalWS()
  n4 = OneMinus(None=Saturate(None=DotProduct(A=@n2, B=@n3)))
  n5 = MaterialFunctionCall(
   TextureObject=TextureObject() {texture=/Game/Weapons/Prototype/ObjectMover/Textures/T_Movetool_Transition}
   TextureSize=VectorParameter() {parameterName=TextureSize, defaultValue=<rgba 64 64 64 0>}.RGB
   WorldPosition=MaterialFunctionCall() {materialFunction=&WorldPositionBehindTranslucency}.Result
  ) {materialFunction=&WorldAlignedTexture}
  n6 = ComponentMask(None=@n5.XYZ Texture) {g=true}
  n7 = ScalarParameter() {parameterName=BlockSineAdd}
  n8 = ConstantBiasScale(None=Sine(None=Add(A=@n6, B=@n7)))
  n9 = ScalarParameter() {defaultValue=0.5, parameterName=BlocksOpacity}
  n10 = Add(A=@n4, B=Add(A=ComponentMask(None=@n5.XYZ Texture) {r=true}, B=Multiply(A=@n8, B=@n9)))
  n11 = LinearInterpolate(
   A=VectorParameter() {parameterName=ColorLow, defaultValue=<rgba 0.0867 0.315 1 1>}.RGB
   B=VectorParameter() {parameterName=ColorHigh, defaultValue=<rgba 0.1303 1 0.3759 1>}.RGB
   Alpha=@n10
  )
  n12 = LinearInterpolate(A=@n11, B=Multiply(A=@n11, B=ScalarParameter() {defaultValue=3, parameterName=EdgeIntensity}), Alpha=@n4)
  n13 = ParticleColor()
  MaterialFunctionCall(Emissive=Multiply(A=ShadingPathSwitch(Default=@n1, Deferred=@n12, Forward=@n1, Mobile=@n1), B=@n13.RGB)) {materialFunction=&MF_FX_EmissiveAdjustments}
  n14 = ScalarParameter() {defaultValue=0.3, parameterName=BlocksBG}
  n15 = ConstantBiasScale(None=TextureCoordinate()) {bias=-0.5, scale=2}
  n16 = SphereMask(
   A=Power(Base=Abs(None=@n15), Exp=ScalarParameter() {defaultValue=1, parameterName=SourceMaskExponent}) {constExponent=1.5}
   B=Constant()
   Radius=Constant() {r=1}
   Hardness=Constant()
  )
  n17 = DynamicParameter() {defaultValue=<rgba 0.5 1 1 1>, paramNames [Center,Radius,Param3,Param4]}
  n18 = ScalarParameter() {defaultValue=0.8346, parameterName=Center}
  n19 = Multiply(
   A=@n16
   B=SphereMask(
    A=@n16
    B=StaticSwitchParameter(True=@n17.Radius, False=@n18) {parameterName=UseDynamicParameterForCenter}
    Radius=ScalarParameter() {defaultValue=0.35, parameterName=Radius}
    Hardness=ScalarParameter() {parameterName=Hardness}
   )
  )
  n20 = SphereMask(
   A=MaterialFunctionCall(Float2=TextureCoordinate()) {materialFunction=&BreakOutFloat2Components}.R
   B=DynamicParameter() {defaultValue=<rgba 0.5 1 1 1>, paramNames [Center,Radius,Param3,Param4]}.Center
   Radius=ScalarParameter() {defaultValue=0.35, parameterName=Radius}
   Hardness=ScalarParameter() {parameterName=Hardness}
  )
  n21 = MaterialFunctionCall!(Distance Bias, Distance Divisor=ScalarParameter() {defaultValue=20, parameterName=Distance Divisor}, Scene Depth, Pixel Depth) {materialFunction=/Game/Effects/Fort_Effects/Materials/_Functions/MF_DistanceBehindTranslucency_01}
  n22 = Multiply(
   A=Multiply(A=Add(A=@n10, B=@n14), B=StaticSwitchParameter(True=@n19, False=@n20) {defaultValue=true, parameterName=UseRadialSweep})
   B=StaticSwitchParameter(True=Ceil(None=@n21.Normalized Inverted), False=@n21.Normalized Inverted) {defaultValue=true, parameterName=CeilDistanceFade}
  )
  n23 = Abs(None=ConstantBiasScale(None=TextureCoordinate()) {bias=-0.5, scale=2})
  n24 = ScalarParameter() {defaultValue=8, parameterName=SquareMaskExponent}
  n25 = Constant()
  n26 = Multiply(A=@n22, B=SphereMask(A=Power(Base=@n23, Exp=@n24), B=@n25, Radius=Constant() {r=0.5}, Hardness=@n25))
  Multiply(
   A=@n13.A
   B=ShadingPathSwitch(Default=@n1, Deferred=StaticSwitchParameter(True=@n26, False=@n22) {parameterName=UseSquareMask}, Forward=@n1, Mobile=@n1)
  )
 ]
}
